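# Physics asset for the HydraTrumpet Coach pickaxe, setting up secondary motion for the decorative bow on the axe.
# A single capsule on the pickaxe body anchors four two-bone chains: left and right bow loops and left and right bow tails, each bone with a small capsule.
# Constraints link each chain root to the body and each tip to its root, with moderate swing and very tight twist limits.
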PhysicsAsset /BRCosmetics/Weapons/FORT_Melee/Pickaxe_HydraTrumpet_Coach/Meshes/HydraTrumpet_Coach_Axe_Physics
bodies {
 base_pickaxeBody [{shapeType=Capsule, length=21.2106, radius=2.5543, shapeName=base_pickaxeBody_capsule, center=<-2.9511 0 0>, rotation=<rot 90 0 0>}]
 dyn_Bow02_l [{shapeType=Capsule, length=2, radius=2, shapeName=dyn_Bow02_l_capsule, center=<2 0 0>, rotation=<rot 90 0 0>}]
 dyn_Bow02_r [{shapeType=Capsule, length=2, radius=2, shapeName=dyn_Bow02_r_capsule, center=<2 0 0>, rotation=<rot 90 0 0>}]
 dyn_Bow_l [{shapeType=Capsule, length=2, radius=2, shapeName=dyn_Bow_l_capsule, center=<1 0 0>, rotation=<rot 90 0 0>}]
 dyn_Bow_r [{shapeType=Capsule, length=2, radius=2, shapeName=dyn_Bow_r_capsule, center=<1 0 0>, rotation=<rot 90 0 0>}]
 dyn_bowTail02_l [{shapeType=Capsule, length=2, radius=2, shapeName=dyn_bowTail02_l_capsule, center=<2 0 0>, rotation=<rot 90 0 0>}]
 dyn_bowTail02_r [{shapeType=Capsule, length=2, radius=2, shapeName=dyn_bowTail02_r_capsule, center=<2 0 0>, rotation=<rot 90 0 0>}]
 dyn_bowTail_l [{shapeType=Capsule, length=2, radius=2, shapeName=dyn_bowTail_l_capsule, center=<1 0 0>, rotation=<rot 90 0 0>}]
 dyn_bowTail_r [{shapeType=Capsule, length=2, radius=2, shapeName=dyn_bowTail_r_capsule, center=<1 0 0>, rotation=<rot 90 0 0>}]
}
constraints [
 {bone1Name=dyn_Bow_l, bone2Name=base_pickaxeBody, swing1LimitDegrees=30, swing2LimitDegrees=20, twistLimitDegrees=2}
 {bone1Name=dyn_Bow02_l, bone2Name=dyn_Bow_l, swing1LimitDegrees=20, swing2LimitDegrees=20, twistLimitDegrees=2}
 {bone1Name=dyn_Bow_r, bone2Name=base_pickaxeBody, swing1LimitDegrees=30, swing2LimitDegrees=20, twistLimitDegrees=2}
 {bone1Name=dyn_Bow02_r, bone2Name=dyn_Bow_r, swing1LimitDegrees=20, swing2LimitDegrees=20, twistLimitDegrees=2}
 {bone1Name=dyn_bowTail_l, bone2Name=base_pickaxeBody, swing1LimitDegrees=30, swing2LimitDegrees=20, twistLimitDegrees=2}
 {bone1Name=dyn_bowTail02_l, bone2Name=dyn_bowTail_l, swing1LimitDegrees=20, swing2LimitDegrees=20, twistLimitDegrees=2}
 {bone1Name=dyn_bowTail_r, bone2Name=base_pickaxeBody, swing1LimitDegrees=30, swing2LimitDegrees=20, twistLimitDegrees=2}
 {bone1Name=dyn_bowTail02_r, bone2Name=dyn_bowTail_r, swing1LimitDegrees=20, swing2LimitDegrees=20, twistLimitDegrees=2}
]
